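# Translucent UI material for a Freeform image loading placeholder, showing a soft animated shimmer.
# A delayed, looping time value passes through a sine ease-in-out curve to move the end of a linear gradient.
# Multiplying the gradient by the inverse of its floor zeroes the saturated region, leaving a moving ramp with a hard edge.
# That ramp blends between two low levels, likely opacity over constant white.
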Material /FortniteCoreUI/UI/Freeform/Art/M_UI_Freeform_ImageLoading
graph {
 properties {
  bAutomaticallySetUsageInEditor=false
  bUsedWithStaticMesh=false
  blendMode=BLEND_Translucent
  materialDomain=MD_UI
  usedShadingModels=Default Lit
 }
 nodes [
  Constant3Vector() {constant=<rgba 1 1 1 0>}
  n1 = Constant() {r=2}
  n2 = Constant() {r=0.125}
  n3 = MaterialFunctionCall(GradientEnd=MaterialFunctionCall("Value (0-1)"=MaterialFunctionCall!(Animation Duration=@n1, Delay=@n2, Reset to Default) {materialFunction=/FortUILibrary/MaterialFunctions/Time/MF_UI_AnimationDelay}.Result) {materialFunction=&MF_UI_EaseCurves}.EaseInOutSine) {materialFunction=&MF_UI_LinearGradient}
  n4 = Floor(None=@n3.Result)
  LinearInterpolate(Alpha=Multiply(A=@n3.Result, B=OneMinus(None=@n4))) {constA=0.2, constB=0.4}
 ]
}
