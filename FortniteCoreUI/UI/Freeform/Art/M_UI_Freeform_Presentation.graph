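# M_UI_Freeform_Presentation is a translucent UI-domain material built to present a Freeform texture.
# It exposes the image as the FreeformTexture object parameter and reads that texture's size through a TextureProperty node.
# MF_UI_UVScaling takes the size and supplies its "Scaled To Fill" UVs to a single texture sample of the same parameter.
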Material /FortniteCoreUI/UI/Freeform/Art/M_UI_Freeform_Presentation
graph {
 properties {
  bAutomaticallySetUsageInEditor=false
  bSubstrateRoughnessTracking=false
  bUsedWithStaticMesh=false
  blendMode=BLEND_Translucent
  materialDomain=MD_UI
  usedShadingModels=Default Lit
 }
 nodes [
  n1 = TextureObjectParameter() {parameterName=FreeformTexture}
  TextureSample(
   UVs=MaterialFunctionCall(TextureSize=TextureProperty(Tex=@n1)) {materialFunction=&MF_UI_UVScaling}.Scaled To Fill
   Tex=@n1
  )
 ]
}
