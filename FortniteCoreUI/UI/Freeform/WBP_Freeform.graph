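# Freeform MOTD interstitial screen: a CommonActivatableWidget with title, scrollable rich text and two buttons beside a CMS image, fading in on activation.
# SetButtons and SetImageURL use only the first array entry, collapsing the call-to-action button when none exists; focus then falls to Dismiss.
# Clicks report to the interstitial viewmodel; the call to action runs its MOTD action.
# Closing dismisses through the viewmodel, requesting the next interstitial only from Dismiss.
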WidgetBlueprint /FortniteCoreUI/UI/Freeform/WBP_Freeform
animations [Anim_Intro {
  binding_count=2
  length_seconds=0.4
  object=&self:Anim_Intro
  sequence {
   bindings [
    Root {
     id=#1
     possessed_class=GridPanel
     tracks [{
       type=MovieSceneFloatTrack
       display_name=Render Opacity
       sections [{type=MovieSceneFloatSection, channels [None {type=MovieSceneScriptingFloatChannel, keys [[0,0],[18,1]]}]}]
      }]
    }
    Stack {
     id=#2
     possessed_class=StackBox
     tracks [
      {
       type=MovieScene2DTransformTrack
       sections [{type=MovieScene2DTransformSection, channels [Translation.Y {type=MovieSceneScriptingFloatChannel, keys [[0,48],[9,48],[24,0]]}]}]
      }
      {
       type=MovieSceneFloatTrack
       display_name=Render Opacity
       sections [{type=MovieSceneFloatSection, end_frame=24, start_frame=0, channels [None {type=MovieSceneScriptingFloatChannel, keys [[0,0],[12,0],[24,1]]}]}]
      }
     ]
    }
   ]
   display_rate {numerator=60}
  }
 }]
graphs [
 BP_GetDesiredFocusTarget {nodes [
   Get Desired Focus Target()
   n1 = Get Button_Loud() -> (Button_Loud:WBP UIKit Button Loud Object Reference)
   > Return Node!(
    execute
    ReturnValue:Widget Object Reference=Select(
     Option 0:Widget Object Reference=Get Button_Regular() -> (Button_Regular:WBP UIKit Button Regular Object Reference)
     Option 1:Widget Object Reference=@n1
     Index=IsVisible(self=@n1)
    ) -> (ReturnValue:Widget Object Reference)
   )
  ]}
 EventGraph {nodes [
   "On Clicked (Button_Loud)"!() -> (OutputDelegate, then, Button:Common Button Base Object Reference) {is=title K2Node_ComponentBoundEvent}
   n1 = Get ButtonAction() -> (ButtonAction:Motd Action Base Object Reference)
   > Branch(Condition=IsValid(Object=@n1))
   > OnButtonPressed!(
    execute
    self:Fort Interstitial VM Object Reference=Get Vertical40_60LayoutInterstitialVM() -> (Vertical40_60LayoutInterstitialVM:Vertical Split Layout Interstitial Viewmodel Object Reference)
    ButtonType:EInterstitialButtonType Enum=ActionButton
    Style:EMotdButtonStyle Enum=Main
    Action:Motd Action Base Object Reference=@n1
   ) -> (then)
   > n2 = Cast To FortPlayerController(Object=GetOwningPlayer())
   > CallToAction!(execute, self:Motd Action Base Object Reference=@n1, PlayerController:Fort Player Controller Object Reference=@n2) -> (then)
   > Branch(Condition=ShouldCloseWidget!(self:Motd Action Base Object Reference=@n1) -> (ReturnValue:bool))
   > On Widget Deactivation!(execute, self, bShowNextOnDismissal:bool=false) -> (then)
   "On Clicked (Button_Regular)"!() -> (OutputDelegate, then, Button:Common Button Base Object Reference) {is=title K2Node_ComponentBoundEvent}
   > OnButtonPressed(self=Get Vertical40_60LayoutInterstitialVM(), ButtonType=Close, Style=Secondary, Action=Get ButtonAction())
   > On Widget Deactivation(bShowNextOnDismissal=true)
   n3 = On Widget Deactivation!() -> (OutputDelegate, then, bShowNextOnDismissal:bool=~) {is=title K2Node_CustomEvent}
   n4 = Get Vertical40_60LayoutInterstitialVM()
   > n5 = Is Valid(InputObject=@n4)
   > n6 = DismissInterstitial!(execute, self:Fort Interstitial VM Object Reference=@n4, bShouldShowNext:bool=@n3.bShowNextOnDismissal) -> (then)
   DeactivateWidget(execute=[@n6.then, @n5.Is Not Valid])
   Event Destruct()
   n7 = Get ScrollBox() -> (ScrollBox:Scroll Box Object Reference)
   > UnregisterScrollRecipientExternal!(execute, self:Common User Widget Object Reference, AnalogScrollRecipient:Widget Object Reference=@n7) -> (then)
   Event Construct()
   > RegisterScrollRecipientExternal!(execute, self:Common User Widget Object Reference, AnalogScrollRecipient:Widget Object Reference=@n7) -> (then)
   Event On Activated()
   > PlayAnimationForward(InAnimation=Get Anim_Intro() -> (Anim_Intro:Widget Animation Object Reference))
  ]}
 SetButtons {nodes [
   n1 = SetButtons!() -> (then, Buttons:Array of Motd Button Structures) {is=title K2Node_FunctionEntry}
   > n2 = Branch(Condition=Is Empty!(TargetArray:Array of Motd Button Structures=@n1) -> (ReturnValue:bool))
   n3 = Get Button_Loud()
   > SetVisibility(self=@n3, InVisibility=Collapsed)
   SetVisibility(execute=@n2.else, self=@n3, InVisibility=Visible)
   n4 = Break Motd Button!(MotdButton:"Motd Button Structure (by ref)"="Get (a ref)"(Array:Array of Motd Button Structures=@n1, Dimension 1=0) -> (Output:"Motd Button Structure (by ref)")) -> (Text:Text, Style:EMotdButtonStyle Enum, Action:Motd Action Base Object Reference) {is=title K2Node_BreakStruct}
   > SetText!(execute, self:Fort CTAButton Object Reference=@n3, ButtonText:"Text (by ref)"=@n4) -> (then)
   > Set ButtonAction(ButtonAction:Motd Action Base Object Reference=@n4.Action) -> (Output_Get:Motd Action Base Object Reference)
  ]}
 SetImageURL {nodes [
   n1 = SetImageURL!() -> (then, MotdImage:Array of Motd Image Structures) {is=title K2Node_FunctionEntry}
   > SetBrushFromURL!(
    execute
    self:Fort Lazy Image Object Reference=Get FortLazyImage() -> (FortLazyImage:Fort Lazy Image Object Reference)
    CMSImageURL:String=Break Motd Image!(MotdImage:"Motd Image Structure (by ref)"="Get (a ref)"(Array:Array of Motd Image Structures=@n1, Dimension 1=0) -> (Output:"Motd Image Structure (by ref)")) -> (Url:String, Width:int, Height:int) {is=title K2Node_BreakStruct}
    bMatchSize:bool=false
    Delay:float=0
    DownloadPriority:ECMSDownloadPriority Enum=VeryHigh
   ) -> (then)
  ]}
]
variables [ButtonAction {type=MotdActionBase, default=null}]
widgets {
 info {inheritedWidgetCount=0, parentClass=&CommonActivatableWidget, rootWidgetClass=&GridPanel, widgetCount=16}
 widgets [
  Root {widgetClassPath=&GridPanel, widget=Root, properties {columnFill [1], rowFill [1]}}
  Background {
   widgetClassPath=&Image
   parent=Root
   widget=Background
   properties {brush {
     bIsDynamicallyLoaded=false
     drawAs=Image
     imageType=FullColor
     mirroring=NoMirror
     resourceName=None
     resourceObject=/FortniteCoreUI/UI/Freeform/Art/MI_UI_Freeform_Background
     tiling=NoTile
     imageSize=<32 32>
     margin=<ltrb 0 0 0 0>
     outlineSettings {
      bUseBrushTransparency=false
      roundingType=HalfHeightRadius
      width=0
      color {colorUseRule=UseColor_Specified, specifiedColor=<rgba 0 0 0 0>}
      cornerRadii=<0 0 0 0>
     }
     tintColor {colorUseRule=UseColor_Specified, specifiedColor=<rgba 1 1 1 1>}
     uVRegion {bIsValid=false, max=<0 0>, min=<0 0>}
    }}
   slot {class=GridSlot, properties {layer=-10}}
  }
  HB_Content {
   widgetClassPath=&HorizontalBox
   parent=Root
   widget=HB_Content
   properties {bOverride_Cursor=true, flowDirectionPreference=Culture}
   slot {class=GridSlot, properties {horizontalAlignment=HAlign_Center, layer=5}}
  }
  O_SideContent {widgetClassPath=&Overlay, parent=HB_Content, widget=O_SideContent, slot {class=HorizontalBoxSlot}}
  CaptureForPostBufferUpdate {
   widgetClassPath=&WBP_CaptureForPostBufferUpdate
   parent=O_SideContent
   widget=CaptureForPostBufferUpdate
   slot {class=OverlaySlot, properties {horizontalAlignment=HAlign_Fill, verticalAlignment=VAlign_Fill}}
  }
  SafeZone {
   widgetClassPath=&SafeZone
   parent=O_SideContent
   widget=SafeZone
   slot {class=OverlaySlot, properties {horizontalAlignment=HAlign_Fill, verticalAlignment=VAlign_Center, padding=<ltrb 80 96 80 72>}}
  }
  SB_Width {
   widgetClassPath=&SizeBox
   parent=SafeZone
   widget=SB_Width
   properties {bOverride_WidthOverride=true, widthOverride=600}
   slot {class=SafeZoneSlot}
  }
  Stack {
   widgetClassPath=&StackBox
   parent=SB_Width
   widget=Stack
   properties {navigation=&self:WidgetTree.Stack.WidgetNavigation_0, orientation=Orient_Vertical}
   slot {class=SizeBoxSlot}
  }
  Title {
   widgetClassPath=&CommonTextBlock
   parent=Stack
   widget=Title
   properties {
    autoWrapText=true
    text=Title
    textTransformPolicy=ToUpper
    font {
     bForceMonospaced=false
     bMaterialIsStencil=false
     fontMaterial=null
     fontObject=&HeadingNow
     letterSpacing=0
     monospacedWidth=1
     size=25.5
     skewAmount=0
     typefaceFontName=86 Bold Italic
     outlineSettings {
      bApplyOutlineToDropShadows=false
      bMiteredCorners=false
      bSeparateFillAlpha=false
      outlineMaterial=null
      outlineSize=0
      outlineColor=<rgba 0 0 0 1>
     }
    }
    shadowOffset=<0 0>
   }
   slot {class=StackBoxSlot, properties {padding=<ltrb 0 0 0 32>}}
  }
  ScrollBox {
   widgetClassPath=&ScrollBox
   bIsVariable=true
   parent=Stack
   widget=ScrollBox
   properties {
    bAnimateWheelScrolling=true
    scrollbarPadding=<ltrb 8 0 0 0>
    scrollbarThickness=<12 12>
    widgetBarStyle {
     thickness=12
     draggedThumbImage {
      bIsDynamicallyLoaded=false
      drawAs=RoundedBox
      imageType=FullColor
      mirroring=NoMirror
      resourceName=../../../Engine/Content/Slate/Common/Scrollbar_Thumb.png
      resourceObject=null
      tiling=NoTile
      imageSize=<12 12>
      margin=<ltrb 0.25 0.25 0.25 0.25>
      outlineSettings {
       bUseBrushTransparency=false
       roundingType=FixedRadius
       width=2
       color {colorUseRule=UseColor_Specified, specifiedColor=<rgba 0 0 0 0>}
       cornerRadii=<6 6 6 6>
      }
      tintColor {colorUseRule=UseColor_Specified, specifiedColor=<rgba 1 1 1 1>}
      uVRegion {bIsValid=false, max=<0 0>, min=<0 0>}
     }
     horizontalBackgroundImage {
      bIsDynamicallyLoaded=false
      drawAs=NoDrawType
      imageType=NoImage
      mirroring=NoMirror
      resourceName=None
      resourceObject=null
      tiling=NoTile
      imageSize=<0 0>
      margin=<ltrb 0 0 0 0>
      outlineSettings {
       bUseBrushTransparency=false
       roundingType=HalfHeightRadius
       width=0
       color {colorUseRule=UseColor_Specified, specifiedColor=<rgba 0 0 0 0>}
       cornerRadii=<0 0 0 0>
      }
      tintColor {colorUseRule=UseColor_Specified, specifiedColor=<rgba 1 1 1 1>}
      uVRegion {bIsValid=false, max=<0 0>, min=<0 0>}
     }
     horizontalBottomSlotImage {
      bIsDynamicallyLoaded=false
      drawAs=NoDrawType
      imageType=FullColor
      mirroring=NoMirror
      resourceName=../../../Engine/Content/Slate/Common/Scrollbar_Background_Horizontal.png
      resourceObject=null
      tiling=NoTile
      imageSize=<8 8>
      margin=<ltrb 0 0 0 0>
      outlineSettings {
       bUseBrushTransparency=false
       roundingType=HalfHeightRadius
       width=0
       color {colorUseRule=UseColor_Specified, specifiedColor=<rgba 0 0 0 0>}
       cornerRadii=<0 0 0 0>
      }
      tintColor {colorUseRule=UseColor_Specified, specifiedColor=<rgba 1 1 1 1>}
      uVRegion {bIsValid=false, max=<0 0>, min=<0 0>}
     }
     horizontalTopSlotImage {
      bIsDynamicallyLoaded=false
      drawAs=NoDrawType
      imageType=FullColor
      mirroring=NoMirror
      resourceName=../../../Engine/Content/Slate/Common/Scrollbar_Background_Horizontal.png
      resourceObject=null
      tiling=NoTile
      imageSize=<8 8>
      margin=<ltrb 0 0 0 0>
      outlineSettings {
       bUseBrushTransparency=false
       roundingType=HalfHeightRadius
       width=0
       color {colorUseRule=UseColor_Specified, specifiedColor=<rgba 0 0 0 0>}
       cornerRadii=<0 0 0 0>
      }
      tintColor {colorUseRule=UseColor_Specified, specifiedColor=<rgba 1 1 1 1>}
      uVRegion {bIsValid=false, max=<0 0>, min=<0 0>}
     }
     hoveredThumbImage {
      bIsDynamicallyLoaded=false
      drawAs=RoundedBox
      imageType=FullColor
      mirroring=NoMirror
      resourceName=../../../Engine/Content/Slate/Common/Scrollbar_Thumb.png
      resourceObject=null
      tiling=NoTile
      imageSize=<12 12>
      margin=<ltrb 0.25 0.25 0.25 0.25>
      outlineSettings {
       bUseBrushTransparency=false
       roundingType=FixedRadius
       width=2
       color {colorUseRule=UseColor_Specified, specifiedColor=<rgba 0 0 0 0>}
       cornerRadii=<6 6 6 6>
      }
      tintColor {colorUseRule=UseColor_Specified, specifiedColor=<rgba 1 1 1 1>}
      uVRegion {bIsValid=false, max=<0 0>, min=<0 0>}
     }
     normalThumbImage {
      bIsDynamicallyLoaded=false
      drawAs=RoundedBox
      imageType=FullColor
      mirroring=NoMirror
      resourceName=../../../Engine/Content/Slate/Common/Scrollbar_Thumb.png
      resourceObject=null
      tiling=NoTile
      imageSize=<12 12>
      margin=<ltrb 0.25 0.25 0.25 0.25>
      outlineSettings {
       bUseBrushTransparency=false
       roundingType=FixedRadius
       width=2
       color {colorUseRule=UseColor_Specified, specifiedColor=<rgba 0 0 0 0>}
       cornerRadii=<6 6 6 6>
      }
      tintColor {colorUseRule=UseColor_Specified, specifiedColor=<rgba 1 1 1 0.749>}
      uVRegion {bIsValid=false, max=<0 0>, min=<0 0>}
     }
     verticalBackgroundImage {
      bIsDynamicallyLoaded=false
      drawAs=Box
      imageType=FullColor
      mirroring=NoMirror
      resourceName=None
      resourceObject=/FortUILibrary/UIKit/ScrollBar/MaterialInstances/MI_UIKit_ScrollbarBackground
      tiling=NoTile
      imageSize=<12 12>
      margin=<ltrb 0.5 0.5 0.5 0.5>
      outlineSettings {
       bUseBrushTransparency=false
       roundingType=HalfHeightRadius
       width=0
       color {colorUseRule=UseColor_Specified, specifiedColor=<rgba 0 0 0 0>}
       cornerRadii=<0 0 0 0>
      }
      tintColor {colorUseRule=UseColor_Specified, specifiedColor=<rgba 0.521 0.521 0.521 0.502>}
      uVRegion {bIsValid=false, max=<0 0>, min=<0 0>}
     }
     verticalBottomSlotImage {
      bIsDynamicallyLoaded=false
      drawAs=NoDrawType
      imageType=FullColor
      mirroring=NoMirror
      resourceName=../../../Engine/Content/Slate/Common/Scrollbar_Background_Vertical.png
      resourceObject=null
      tiling=NoTile
      imageSize=<8 8>
      margin=<ltrb 0 0 0 0>
      outlineSettings {
       bUseBrushTransparency=false
       roundingType=HalfHeightRadius
       width=0
       color {colorUseRule=UseColor_Specified, specifiedColor=<rgba 0 0 0 0>}
       cornerRadii=<0 0 0 0>
      }
      tintColor {colorUseRule=UseColor_Specified, specifiedColor=<rgba 1 1 1 1>}
      uVRegion {bIsValid=false, max=<0 0>, min=<0 0>}
     }
     verticalTopSlotImage {
      bIsDynamicallyLoaded=false
      drawAs=NoDrawType
      imageType=FullColor
      mirroring=NoMirror
      resourceName=../../../Engine/Content/Slate/Common/Scrollbar_Background_Vertical.png
      resourceObject=null
      tiling=NoTile
      imageSize=<8 8>
      margin=<ltrb 0 0 0 0>
      outlineSettings {
       bUseBrushTransparency=false
       roundingType=HalfHeightRadius
       width=0
       color {colorUseRule=UseColor_Specified, specifiedColor=<rgba 0 0 0 0>}
       cornerRadii=<0 0 0 0>
      }
      tintColor {colorUseRule=UseColor_Specified, specifiedColor=<rgba 1 1 1 1>}
      uVRegion {bIsValid=false, max=<0 0>, min=<0 0>}
     }
    }
    widgetStyle {
     barThickness=12
     bottomShadowBrush {
      bIsDynamicallyLoaded=false
      drawAs=NoDrawType
      imageType=FullColor
      mirroring=NoMirror
      resourceName=../../../Engine/Content/Slate/Common/ScrollBoxShadowBottom.png
      resourceObject=null
      tiling=NoTile
      imageSize=<16 8>
      margin=<ltrb 0.5 0 0.5 1>
      outlineSettings {
       bUseBrushTransparency=false
       roundingType=HalfHeightRadius
       width=0
       color {colorUseRule=UseColor_Specified, specifiedColor=<rgba 0 0 0 0>}
       cornerRadii=<0 0 0 0>
      }
      tintColor {colorUseRule=UseColor_Specified, specifiedColor=<rgba 1 1 1 1>}
      uVRegion {bIsValid=false, max=<0 0>, min=<0 0>}
     }
     horizontalScrolledContentPadding=<ltrb 0 0 1 0>
     leftShadowBrush {
      bIsDynamicallyLoaded=false
      drawAs=NoDrawType
      imageType=FullColor
      mirroring=NoMirror
      resourceName=../../../Engine/Content/Slate/Common/ScrollBoxShadowLeft.png
      resourceObject=null
      tiling=NoTile
      imageSize=<8 16>
      margin=<ltrb 1 0.5 0 0.5>
      outlineSettings {
       bUseBrushTransparency=false
       roundingType=HalfHeightRadius
       width=0
       color {colorUseRule=UseColor_Specified, specifiedColor=<rgba 0 0 0 0>}
       cornerRadii=<0 0 0 0>
      }
      tintColor {colorUseRule=UseColor_Specified, specifiedColor=<rgba 1 1 1 1>}
      uVRegion {bIsValid=false, max=<0 0>, min=<0 0>}
     }
     rightShadowBrush {
      bIsDynamicallyLoaded=false
      drawAs=NoDrawType
      imageType=FullColor
      mirroring=NoMirror
      resourceName=../../../Engine/Content/Slate/Common/ScrollBoxShadowRight.png
      resourceObject=null
      tiling=NoTile
      imageSize=<8 16>
      margin=<ltrb 0 0.5 1 0.5>
      outlineSettings {
       bUseBrushTransparency=false
       roundingType=HalfHeightRadius
       width=0
       color {colorUseRule=UseColor_Specified, specifiedColor=<rgba 0 0 0 0>}
       cornerRadii=<0 0 0 0>
      }
      tintColor {colorUseRule=UseColor_Specified, specifiedColor=<rgba 1 1 1 1>}
      uVRegion {bIsValid=false, max=<0 0>, min=<0 0>}
     }
     topShadowBrush {
      bIsDynamicallyLoaded=false
      drawAs=NoDrawType
      imageType=FullColor
      mirroring=NoMirror
      resourceName=../../../Engine/Content/Slate/Common/ScrollBoxShadowTop.png
      resourceObject=null
      tiling=NoTile
      imageSize=<16 8>
      margin=<ltrb 0.5 1 0.5 0>
      outlineSettings {
       bUseBrushTransparency=false
       roundingType=HalfHeightRadius
       width=0
       color {colorUseRule=UseColor_Specified, specifiedColor=<rgba 0 0 0 0>}
       cornerRadii=<0 0 0 0>
      }
      tintColor {colorUseRule=UseColor_Specified, specifiedColor=<rgba 1 1 1 1>}
      uVRegion {bIsValid=false, max=<0 0>, min=<0 0>}
     }
     verticalScrolledContentPadding=<ltrb 0 0 0 1>
    }
   }
   slot {class=StackBoxSlot, properties {padding=<ltrb 0 0 0 16>, size {sizeRule=Fill, value=1}}}
  }
  RichTextBlock {
   widgetClassPath=/Script/UMG.RichTextBlock
   bIsVariable=true
   parent=ScrollBox
   widget=RichTextBlock
   properties {
    applyLineHeightToBottomLine=false
    autoWrapText=true
    lineHeightPercentage=1.25
    text=Body text
    textStyleSet=/FortniteCoreUI/UI/Freeform/Data/DT_FreeformRichTextStyles
   }
   slot {class=ScrollBoxSlot}
  }
  Button_Loud {
   widgetClassPath=/FortUILibrary/UIKit/Buttons/WBP_UIKit_Button_Loud
   bIsVariable=true
   parent=Stack
   widget=Button_Loud
   properties {
    bDisplayInputActionWhenNotInteractable=false
    bShouldUseFallbackDefaultInputAction=false
    navigation=&self:WidgetTree.Button_Loud.WidgetNavigation_0
    text=Call to action
    visibility=Visible
   }
   slot {class=StackBoxSlot, properties {verticalAlignment=VAlign_Top, padding=<ltrb 0 24 0 0>}}
  }
  Button_Regular {
   widgetClassPath=&WBP_UIKit_Button_Regular
   bIsVariable=true
   parent=Stack
   widget=Button_Regular
   properties {
    bDisplayInputActionWhenNotInteractable=false
    bShouldUseFallbackDefaultInputAction=false
    navigation=&self:WidgetTree.Button_Regular.WidgetNavigation_0
    text=Dismiss
   }
   slot {class=StackBoxSlot, properties {verticalAlignment=VAlign_Top, padding=<ltrb 0 24 0 0>}}
  }
  O_Imagery {
   widgetClassPath=&Overlay
   parent=HB_Content
   widget=O_Imagery
   slot {class=HorizontalBoxSlot, properties {size {sizeRule=Fill, value=1}}}
  }
  SpacerMinSizeCMSImage {
   widgetClassPath=&Spacer
   parent=O_Imagery
   widget=SpacerMinSizeCMSImage
   properties {size=<1160 1080>}
   slot {class=OverlaySlot, properties {horizontalAlignment=HAlign_Fill, verticalAlignment=VAlign_Fill}}
  }
  FortLazyImage {
   widgetClassPath=&FortLazyImage
   bIsVariable=true
   parent=O_Imagery
   widget=FortLazyImage
   properties {
    materialTextureParamName=FreeformTexture
    brush {
     bIsDynamicallyLoaded=false
     drawAs=Image
     imageType=FullColor
     mirroring=NoMirror
     resourceName=None
     resourceObject=/FortniteCoreUI/UI/Freeform/Art/M_UI_Freeform_Presentation
     tiling=NoTile
     imageSize=<1160 1080>
     margin=<ltrb 0 0 0 0>
     outlineSettings {
      bUseBrushTransparency=false
      roundingType=HalfHeightRadius
      width=0
      color {colorUseRule=UseColor_Specified, specifiedColor=<rgba 0 0 0 0>}
      cornerRadii=<0 0 0 0>
     }
     tintColor {colorUseRule=UseColor_Specified, specifiedColor=<rgba 1 1 1 1>}
     uVRegion {bIsValid=false, max=<0 0>, min=<0 0>}
    }
    loadingBackgroundBrush {
     bIsDynamicallyLoaded=false
     drawAs=Image
     imageType=FullColor
     mirroring=NoMirror
     resourceName=None
     resourceObject=/FortniteCoreUI/UI/Freeform/Art/M_UI_Freeform_ImageLoading
     tiling=NoTile
     imageSize=<1160 1080>
     margin=<ltrb 0 0 0 0>
     outlineSettings {
      bUseBrushTransparency=false
      roundingType=HalfHeightRadius
      width=0
      color {colorUseRule=UseColor_Specified, specifiedColor=<rgba 0 0 0 0>}
      cornerRadii=<0 0 0 0>
     }
     tintColor {colorUseRule=UseColor_Specified, specifiedColor=<rgba 1 1 1 1>}
     uVRegion {bIsValid=false, max=<0 0>, min=<0 0>}
    }
   }
   slot {class=OverlaySlot, properties {horizontalAlignment=HAlign_Fill, verticalAlignment=VAlign_Fill}}
  }
 ]
}
